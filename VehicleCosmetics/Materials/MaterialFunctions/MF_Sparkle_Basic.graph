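# Material function for vehicle cosmetics that adds view-dependent glitter sparkles to base color and emissive.
# Multiplies a noise texture sampled by the transformed reflection vector plus an offset with the same noise on tiled UVs, then scales and floors it into a hard sparkle mask.
# The mask brightens base color multiplicatively and blends into emissive.
# Static switches toggle sparkles and optionally limit them to a selected RGB mask channel.
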MaterialFunction /VehicleCosmetics/Materials/MaterialFunctions/MF_Sparkle_Basic
graph {
 properties {description="MF that provides sparkles multiplied over BC, and masked for emissive additions."}
 nodes [
  FunctionOutput(None=NamedRerouteUsage()) {outputName=BCSparkle_Out}
  FunctionOutput(None=NamedRerouteUsage()) {outputName=EmissiveSparkle_Out, sortPriority=1}
  NamedRerouteDeclaration(None=FunctionInput(Preview=Constant3Vector() {constant=<rgba 0.25 0.25 0.25 0>}) {bUsePreviewValueAsDefault=true, inputName=BC_In, sortPriority=1}) {name=BCBase_Out, nodeColor=<rgba 0.0095 0 1 1>}
  NamedRerouteDeclaration(None=FunctionInput(Preview=Constant3Vector() {constant=<rgba 0.5 0.5 1 0>}) {bUsePreviewValueAsDefault=true, inputName=Normal_In, sortPriority=2}) {name=Normal_Out, nodeColor=<rgba 0.509 0.1352 1 1>}
  NamedRerouteDeclaration(None=FunctionInput(Preview=Constant3Vector()) {bUsePreviewValueAsDefault=true, inputName=Emissive_In, sortPriority=3}) {name=EmissiveBase_Out, nodeColor=<rgba 0.8209 1 0 1>}
  n1 = ReflectionVectorWS()
  n2 = ComponentMask(None=Add(A=Transform(None=@n1), B=NamedRerouteUsage())) {g=true, r=true}
  n3 = TextureCoordinate()
  n4 = VectorParameter() {
   desc="Higher tiling values == denser sparkles."
   group=05_Shader Effects
   parameterName=SparkleTiling
   sortPriority=41
   defaultValue=<rgba 4 4 0 1>
  }
  n5 = AppendVector(A=@n4.R, B=@n4.R)
  n6 = Multiply(
   A=TextureSample(UVs=Multiply(A=@n2) {constB=0.0625}) {samplerSource=SSM_Wrap_WorldGroupSettings, texture=/VehicleCosmetics/Textures/UniversalTextures/T_Noise_Generic_01_RGB}.B
   B=TextureSample(UVs=Multiply(A=@n3, B=@n5)) {samplerSource=SSM_Wrap_WorldGroupSettings, texture=&T_Noise_Generic_01_RGB}.B
  )
  n7 = ScalarParameter() {defaultValue=4096, desc="Lower values == fewer standout sparkles.", group=05_Shader Effects, parameterName=SparkleIntensity, sortPriority=49}
  n8 = Floor(None=Add(A=Multiply(A=@n6, B=@n7) {constB=4096}) {constB=-0.03, desc=Darken/Smooth aliased edges}) {desc=Further cleaning glitter artifacting}
  NamedRerouteDeclaration(None=@n8) {name=SparkleMaskEmis_Out, nodeColor=<rgba 0.1016 1 0 1>}
  n9 = NamedRerouteUsage()
  n10 = LinearInterpolate(A=@n9, B=NamedRerouteUsage(), Alpha=NamedRerouteUsage())
  n11 = NamedRerouteUsage()
  n12 = StaticSwitchParameter(
   True=LinearInterpolate(A=@n9, B=@n10, Alpha=ChannelMaskParameter(None=@n11) {group=05_Shader Effects, parameterName=SparkleMaskChannelSelect, sortPriority=46}.R)
   False=@n10
  ) {
   desc="True == Mask by specific channel. Useful if using an existing ColorMask."
   group=05_Shader Effects
   parameterName=Sparkle_MaskByChannel
   sortPriority=44
  }
  NamedRerouteDeclaration(None=StaticSwitchParameter(True=@n12, False=@n9) {desc=Enable Sparkles, group=05_Shader Effects, parameterName=Sparkles, sortPriority=40}) {name=Emissive_Out, nodeColor=<rgba 0 1 0.2266 1>}
  NamedRerouteDeclaration(None=FunctionInput(Preview=Constant3Vector()) {bUsePreviewValueAsDefault=true, inputName=RGBMask_In, sortPriority=4}) {name=RGBMask_Out, nodeColor=<rgba 1 0.0447 0 1>}
  n13 = NamedRerouteUsage()
  n14 = NamedRerouteUsage()
  n15 = Multiply(A=Saturate(None=@n8), B=@n14)
  n16 = StaticSwitchParameter(
   True=LinearInterpolate(
    A=NamedRerouteUsage()
    B=NamedRerouteUsage()
    Alpha=ChannelMaskParameter(None=@n13) {group=05_Shader Effects, parameterName=SparkleMaskChannelSelect, sortPriority=46}.R
   )
   False=NamedRerouteDeclaration(None=Add(A=@n15, B=@n14)) {name=SparkleBC_Out, nodeColor=<rgba 0 0.4375 1 1>}
  ) {
   desc="True == Mask by specific channel. Useful if using an existing ColorMask."
   group=05_Shader Effects
   parameterName=Sparkle_MaskByChannel
   sortPriority=44
  }
  NamedRerouteDeclaration(None=StaticSwitchParameter(True=@n16, False=@n14) {desc=Enable Sparkles, group=05_Shader Effects, parameterName=Sparkles, sortPriority=40}) {name=BCSparkle_Out, nodeColor=<rgba 0 1 0.2266 1>}
 ]
}
