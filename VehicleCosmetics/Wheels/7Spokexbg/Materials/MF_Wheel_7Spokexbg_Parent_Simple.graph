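# Material function for the simple parent material of the 7Spokexbg cosmetic vehicle wheel, adding a sparkle effect to the shared wheel look.
# MF_WheelParent_Simple supplies material attributes, a vertex color red channel and a rim RGBA mask through named reroutes.
# MF_Sparkle_Basic takes extracted attributes and a mask, producing sparkle base color and emissive.
# The output blends an attribute set with overridden color channels against another set using an alpha input.
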MaterialFunction /VehicleCosmetics/Wheels/7Spokexbg/Materials/MF_Wheel_7Spokexbg_Parent_Simple
graph {nodes [
  n1 = NamedRerouteUsage()
  n2 = NamedRerouteUsage()
  n3 = NamedRerouteUsage()
  FunctionOutput(None=BlendMaterialAttributes(
    A=SetMaterialAttributes!(MaterialAttributes=@n1, BaseColor=@n2, EmissiveColor=@n3) {attributeSetTypes [#1, #2]}
    B=NamedRerouteUsage()
    Alpha=NamedRerouteUsage()
   )) {outputName=MaterialAttributes}
  n4 = MaterialFunctionCall() {materialFunction=&MF_WheelParent_Simple}
  NamedRerouteDeclaration(None=@n4.MaterialAttributes) {name=MF_Simple_MatAtt_Out, nodeColor=<rgba 0.4288 0.4288 1 1>}
  NamedRerouteDeclaration(None=@n4.VertexColorR) {name=VertexColorR_Out, nodeColor=<rgba 1 0 0 1>}
  NamedRerouteDeclaration(None=@n4.RimRGBAMask) {name=RGBAMask_Out, nodeColor=<rgba 1 0 0.2578 1>}
  n5 = GetMaterialAttributes(None=NamedRerouteUsage()) {attributeGetTypes [#1, #3, #2]}
  n6 = MaterialFunctionCall!(BC_In=@n5.BaseColor, Normal_In=@n5.BaseColor, Emissive_In=@n5.BaseColor, RGBMask_In=NamedRerouteUsage()) {materialFunction=/VehicleCosmetics/Materials/MaterialFunctions/MF_Sparkle_Basic}
  NamedRerouteDeclaration(None=@n6.BCSparkle_Out) {name=BC_Out, nodeColor=<rgba 0 0.5078 1 1>}
  NamedRerouteDeclaration(None=@n6.EmissiveSparkle_Out) {name=Emissive_Out, nodeColor=<rgba 0.8281 1 0 1>}
 ]}
